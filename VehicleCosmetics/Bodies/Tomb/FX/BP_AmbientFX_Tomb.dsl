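; BP_AmbientFX_Tomb is a Blueprint for the ambient effect of the Tomb vehicle body cosmetic.
; On BeginPlay it activates its AmbientFX component with reset, and on EndPlay it deactivates that component.
; Its construction script is empty.
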(event EventBeginPlay
  (Components|Activation|Activate (Variables|Default|GetAmbientFX) true))

(event EventEndPlay
  (params (EndPlayReason :type EEndPlayReason%20Enum :default "Destroyed"))
  (Components|Activation|Deactivate (Variables|Default|GetAmbientFX)))

(fn ConstructionScript)
